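# Physics asset for the Modern Witch female medium face accessory, defining bodies and joints for secondary motion of hat and hair.
# Spine, clavicle and head bodies supply torso and head volumes, including a thin box on spine_04.
# Limited joints tie four hat sides and a two-link center chain to the head.
# Short hair chains hang from free-jointed head masters, while a longer six-link left chain anchors to spine_05.
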PhysicsAsset /BRCosmetics/Characters/Player/Female/Medium/Bodies/F_MED_Modern_Witch/Meshes/Parts/F_MED_Modern_Witch_FaceAcc_FE_Dyn
bodies {
 clavicle_l [{shapeType=Capsule, length=11.649, radius=7.5173, center=<9.7417 -0.6451 -3.1117>, rotation=<rot 79.942 0 0>}]
 clavicle_r [{shapeType=Capsule, length=11.3488, radius=7.8175, center=<-9.7417 0.6451 3.1117>, rotation=<rot 79.942 0 0>}]
 dyn_hair_01_bk [{shapeType=Capsule, length=5, radius=1, shapeName=dyn_hair_01_bk_capsule, center=<1.8322 0 0>, rotation=<rot 0 0 90>}]
 dyn_hair_01_l [{shapeType=Capsule, length=8, radius=1, shapeName=dyn_hair_01_l_capsule, rotation=<rot 0 0 90>}]
 dyn_hair_01_r [{shapeType=Box, extentX=7.25, extentY=7.5, extentZ=1.375, center=<4.1392 0.0001 -0.0001>}]
 dyn_hair_02_bk [{shapeType=Capsule, length=5, radius=1, shapeName=dyn_hair_02_bk_capsule, center=<1.8322 0 0>, rotation=<rot 0 0 90>}]
 dyn_hair_02_l [{shapeType=Capsule, length=8, radius=1, shapeName=dyn_hair_02_l_capsule, rotation=<rot 0 0 90>}]
 dyn_hair_02_r [{shapeType=Box, extentX=5.125, extentY=7.5, extentZ=1.375, center=<2.14 -0.0001 -0.0001>}]
 dyn_hair_04_l [
  {shapeType=Capsule, length=3.0938, radius=1.25, center=<6 0 -1.2>, rotation=<rot 90 0 0>}
  {shapeType=Capsule, length=3.0938, radius=1.25, center=<6 0 1.2>, rotation=<rot 90 0 0>}
 ]
 dyn_hair_05_l [
  {shapeType=Capsule, length=5.9974, radius=1.6644, center=<3.7 0 -2>, rotation=<rot 90 0 0>}
  {shapeType=Capsule, length=5.9974, radius=1.6644, center=<3.7 0 2>, rotation=<rot 90 0 0>}
 ]
 dyn_hair_06_l [{shapeType=Capsule, length=5.6481, radius=1.6011, center=<3.7781 0 0>, rotation=<rot 90 0 0>}]
 dyn_hair_07_l [{shapeType=Capsule, length=5.3879, radius=1.8612, center=<3.7781 0 0>, rotation=<rot 90 0 0>}]
 dyn_hair_08_l [{shapeType=Capsule, length=4.9581, radius=1.7949, center=<3.1867 0 0>, rotation=<rot 90 0 0>}]
 dyn_hair_09_l [{shapeType=Capsule, length=3.75, radius=1.25, center=<3.7781 0 0>, rotation=<rot 90 0 0>}]
 dyn_hair_bk_MSTR [{shapeType=Capsule, length=0, radius=0.5, rotation=<rot 90 0 0>}]
 dyn_hair_l_MSTR [{shapeType=Capsule, length=0, radius=0.5, rotation=<rot 90 0 0>}]
 dyn_hair_r_MSTR [{shapeType=Capsule, length=0, radius=0.5, rotation=<rot 90 0 0>}]
 dyn_hat_bk [{shapeType=Capsule, length=2, radius=8, shapeName=dyn_hat_bk_capsule, center=<0.4186 0 -3.6636>}]
 dyn_hat_center_01 [{shapeType=Capsule, length=4.0312, radius=3.5625, center=<5.5 0 0>, rotation=<rot 90 0 0>}]
 dyn_hat_center_02 [{shapeType=Capsule, length=4.0312, radius=3.5625, center=<5.5 0 0>, rotation=<rot 90 0 0>}]
 dyn_hat_fr [{shapeType=Capsule, length=2, radius=8, shapeName=dyn_hat_fr_capsule, center=<0.4186 0 -3.6636>}]
 dyn_hat_l [{shapeType=Capsule, length=2, radius=8, shapeName=dyn_hat_l_capsule, center=<0.4186 0 -3.6636>}]
 dyn_hat_r [{shapeType=Capsule, length=2, radius=8, shapeName=dyn_hat_r_capsule, center=<1.3771 0 -3.6636>}]
 head [{shapeType=Capsule, length=6.8438, radius=6.75, center=<1.0047 0.4234 0>, rotation=<rot 90 0 0>}]
 spine_01 [{shapeType=Capsule, length=13.3419, radius=9, center=<1.8347 0.7008 0>}]
 spine_02 [{shapeType=Capsule, length=10.5919, radius=9, center=<1.2256 1.3756 0>}]
 spine_03 [{shapeType=Capsule, length=14.8731, radius=9, center=<-0.0338 1.88 0>}]
 spine_04 [
  {shapeType=Capsule, length=24.2507, radius=9, center=<-1.7919 3.2113 0>}
  {shapeType=Box, extentX=23.625, extentY=24.25, extentZ=0.8125, center=<-0.0002 18.4606 -13.2056>, rotation=<rot 0 0 16.2921>}
 ]
 spine_05 [{shapeType=Capsule, length=13.5122, radius=8.4253, center=<1.2256 -1.6708 0>}]
}
constraints [
 {
  bone1Name=dyn_hair_l_MSTR
  bone2Name=head
  swing1LimitDegrees=45
  swing1Motion=Free
  swing2LimitDegrees=45
  swing2Motion=Free
  twistLimitDegrees=45
  twistMotion=Free
 }
 {bone1Name=dyn_hair_01_l, bone2Name=dyn_hair_l_MSTR, swing1LimitDegrees=5, swing2LimitDegrees=5, twistLimitDegrees=5}
 {bone1Name=dyn_hair_02_l, bone2Name=dyn_hair_01_l, swing1LimitDegrees=5, swing2LimitDegrees=5, twistLimitDegrees=5}
 {bone1Name=dyn_hair_04_l, bone2Name=spine_05, swing1LimitDegrees=10, swing2LimitDegrees=10, twistLimitDegrees=5}
 {bone1Name=dyn_hair_05_l, bone2Name=dyn_hair_04_l, swing1LimitDegrees=25, swing2LimitDegrees=10, twistLimitDegrees=5}
 {bone1Name=dyn_hair_06_l, bone2Name=dyn_hair_05_l, swing1LimitDegrees=20, swing2LimitDegrees=20, twistLimitDegrees=10}
 {bone1Name=dyn_hair_07_l, bone2Name=dyn_hair_06_l, swing1LimitDegrees=20, swing2LimitDegrees=20, twistLimitDegrees=10}
 {bone1Name=dyn_hair_08_l, bone2Name=dyn_hair_07_l, swing1LimitDegrees=20, swing2LimitDegrees=20, twistLimitDegrees=10}
 {bone1Name=dyn_hair_09_l, bone2Name=dyn_hair_08_l, swing1LimitDegrees=20, swing2LimitDegrees=20, twistLimitDegrees=10}
 {bone1Name=dyn_hat_fr, bone2Name=head, swing1LimitDegrees=5, swing2LimitDegrees=30, twistLimitDegrees=15}
 {bone1Name=dyn_hat_l, bone2Name=head, swing1LimitDegrees=5, swing2LimitDegrees=30, twistLimitDegrees=15}
 {bone1Name=dyn_hat_bk, bone2Name=head, swing1LimitDegrees=5, swing2LimitDegrees=30, twistLimitDegrees=15}
 {bone1Name=dyn_hat_r, bone2Name=head, swing1LimitDegrees=5, swing2LimitDegrees=30, twistLimitDegrees=15}
 {bone1Name=dyn_hat_center_01, bone2Name=head, swing1LimitDegrees=10, swing2LimitDegrees=10, twistLimitDegrees=5}
 {bone1Name=dyn_hat_center_02, bone2Name=dyn_hat_center_01, swing1LimitDegrees=10, swing2LimitDegrees=10, twistLimitDegrees=5}
 {
  bone1Name=dyn_hair_bk_MSTR
  bone2Name=head
  swing1LimitDegrees=45
  swing1Motion=Free
  swing2LimitDegrees=45
  swing2Motion=Free
  twistLimitDegrees=45
  twistMotion=Free
 }
 {bone1Name=dyn_hair_01_bk, bone2Name=dyn_hair_bk_MSTR, swing1LimitDegrees=20, swing2LimitDegrees=20, twistLimitDegrees=5}
 {bone1Name=dyn_hair_02_bk, bone2Name=dyn_hair_01_bk, swing1LimitDegrees=20, swing2LimitDegrees=20, twistLimitDegrees=5}
 {
  bone1Name=dyn_hair_r_MSTR
  bone2Name=head
  swing1LimitDegrees=45
  swing1Motion=Free
  swing2LimitDegrees=45
  swing2Motion=Free
  twistLimitDegrees=45
  twistMotion=Free
 }
 {bone1Name=dyn_hair_01_r, bone2Name=dyn_hair_r_MSTR, swing1LimitDegrees=20, swing2LimitDegrees=20, twistLimitDegrees=5}
 {bone1Name=dyn_hair_02_r, bone2Name=dyn_hair_01_r, swing1LimitDegrees=20, swing2LimitDegrees=20, twistLimitDegrees=5}
 {
  bone1Name=clavicle_l
  bone2Name=spine_05
  swing1LimitDegrees=45
  swing1Motion=Free
  swing2LimitDegrees=45
  swing2Motion=Free
  twistLimitDegrees=45
  twistMotion=Free
 }
 {
  bone1Name=clavicle_r
  bone2Name=spine_05
  swing1LimitDegrees=45
  swing1Motion=Free
  swing2LimitDegrees=45
  swing2Motion=Free
  twistLimitDegrees=45
  twistMotion=Free
 }
 {
  bone1Name=head
  bone2Name=spine_05
  swing1LimitDegrees=45
  swing1Motion=Free
  swing2LimitDegrees=45
  swing2Motion=Free
  twistLimitDegrees=45
  twistMotion=Free
 }
 {
  bone1Name=spine_05
  bone2Name=spine_04
  swing1LimitDegrees=45
  swing1Motion=Free
  swing2LimitDegrees=45
  swing2Motion=Free
  twistLimitDegrees=45
  twistMotion=Free
 }
 {
  bone1Name=spine_04
  bone2Name=spine_03
  swing1LimitDegrees=45
  swing1Motion=Free
  swing2LimitDegrees=45
  swing2Motion=Free
  twistLimitDegrees=45
  twistMotion=Free
 }
 {
  bone1Name=spine_03
  bone2Name=spine_02
  swing1LimitDegrees=45
  swing1Motion=Free
  swing2LimitDegrees=45
  swing2Motion=Free
  twistLimitDegrees=45
  twistMotion=Free
 }
 {
  bone1Name=spine_02
  bone2Name=spine_01
  swing1LimitDegrees=45
  swing1Motion=Free
  swing2LimitDegrees=45
  swing2Motion=Free
  twistLimitDegrees=45
  twistMotion=Free
 }
]
