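# Post-copy physics AnimBP for the Modern Witch face accessory, adding secondary motion to the hat and its straps.
# Copies the attached parent's pose, applies bone constraints to hair and strap bones, then caches it as PreSim.
# bIsInFrontEnd selects a RigidBody-only front-end simulation or a RigidBody plus hat AnimDynamics chain, followed by shared strap AnimDynamics.
# The event graph updates GravityOverride through a deprecated function, but no visible node reads it.
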AnimBlueprint /BRCosmetics/Characters/Player/Female/Medium/Bodies/F_MED_Modern_Witch/Meshes/Parts/F_MED_Modern_Witch_FaceAcc_AnimBP
target_skeleton=&/F_MED_Modern_Witch_FaceAcc_Skeleton
graphs [
 AnimGraph {nodes [
   n1 = RigidBody(ComponentPose=Local To Component(LocalPose=Use cached pose 'PreSim'()), Alpha=0.75) {
    bClampLinearTranslationLimitToRefPose=true
    bOverrideWorldGravity=true
    overridePhysicsAsset=&/F_MED_Modern_Witch_FaceAcc_Dyn
    simulationSpace=BaseBoneSpace
    alphaBoolBlend {blendOption=Linear}
    baseBoneRef {boneName=spine_05}
    componentAppliedLinearAccClamp=<3000 3000 3000>
    componentLinearAccScale=<0.4 0.3 0.3>
    componentLinearVelScale=<0.4 0.3 0.3>
    overrideWorldGravity {z=-980}
   }
   n2 = Anim Dynamics - Bone: dyn_hat_fr!(ComponentPose=@n1, Alpha=0.4, GravityOverride:Vector="0,0,-980", ExternalForce="0,0,0") -> (Pose) {
    angularDampingOverride=0.9999
    bGravityOverrideInSimSpace=true
    bOverrideAngularDamping=true
    bUseGravityOverride=true
    gravityScale=0.45
    simulationSpace=RootRelative
    alphaBoolBlend {blendOption=Linear}
    boundBone {boneName=dyn_hat_fr}
    componentAppliedLinearAccClamp=<5000 5000 5000>
    componentLinearAccScale=<1 1 1>
    componentLinearVelScale=<1 1 1>
    physicsBodyDefinitions [{
      sphereCollisionRadius=0
      boundBone {boneName=dyn_hat_fr}
      boxExtents=<20 20 10>
      constraintSetup {angularLimitsMax=<45 20 45>, angularLimitsMin=<-45 -125 -45>, angularTarget=<0 0 0>}
      localJointOffset=<0 0 -2>
     }]
    relativeSpaceBone {boneName=root}
   }
   n3 = Anim Dynamics - Bone: dyn_hat_l!(ComponentPose=@n2, Alpha=0.2, GravityOverride:Vector="0,0,-980", ExternalForce="0,0,0") -> (Pose) {
    angularDampingOverride=0.9999
    bGravityOverrideInSimSpace=true
    bOverrideAngularDamping=true
    bUseGravityOverride=true
    gravityScale=0.45
    simulationSpace=RootRelative
    alphaBoolBlend {blendOption=Linear}
    boundBone {boneName=dyn_hat_l}
    componentAppliedLinearAccClamp=<5000 5000 5000>
    componentLinearAccScale=<1 1 1>
    componentLinearVelScale=<1 1 1>
    physicsBodyDefinitions [{
      sphereCollisionRadius=0
      boundBone {boneName=dyn_hat_l}
      boxExtents=<20 20 10>
      constraintSetup {angularLimitsMax=<45 20 45>, angularLimitsMin=<-45 -125 -45>, angularTarget=<0 0 0>}
      localJointOffset=<0 0 -2>
     }]
    relativeSpaceBone {boneName=root}
   }
   n4 = Anim Dynamics - Bone: dyn_hat_bk!(ComponentPose=@n3, Alpha=0.4, GravityOverride:Vector="0,0,-980", ExternalForce="0,0,0") -> (Pose) {
    angularDampingOverride=0.9999
    bGravityOverrideInSimSpace=true
    bOverrideAngularDamping=true
    bUseGravityOverride=true
    gravityScale=0.45
    simulationSpace=RootRelative
    alphaBoolBlend {blendOption=Linear}
    boundBone {boneName=dyn_hat_bk}
    componentAppliedLinearAccClamp=<5000 5000 5000>
    componentLinearAccScale=<1 1 1>
    componentLinearVelScale=<1 1 1>
    physicsBodyDefinitions [{
      sphereCollisionRadius=0
      boundBone {boneName=dyn_hat_bk}
      boxExtents=<20 20 10>
      constraintSetup {angularLimitsMax=<45 20 45>, angularLimitsMin=<-45 -125 -45>, angularTarget=<0 0 0>}
      localJointOffset=<0 0 -2>
     }]
    relativeSpaceBone {boneName=root}
   }
   n5 = Anim Dynamics - Bone: dyn_hat_r!(ComponentPose=@n4, Alpha=0.23, GravityOverride:Vector="0,0,-980", ExternalForce="0,0,0") -> (Pose) {
    angularDampingOverride=0.9999
    bGravityOverrideInSimSpace=true
    bOverrideAngularDamping=true
    bUseGravityOverride=true
    gravityScale=0.45
    simulationSpace=RootRelative
    alphaBoolBlend {blendOption=Linear}
    boundBone {boneName=dyn_hat_r}
    componentAppliedLinearAccClamp=<5000 5000 5000>
    componentLinearAccScale=<1.5 1.5 1.5>
    componentLinearVelScale=<1.5 1.5 1.5>
    physicsBodyDefinitions [{
      sphereCollisionRadius=0
      boundBone {boneName=dyn_hat_r}
      boxExtents=<20 20 10>
      constraintSetup {angularLimitsMax=<45 20 45>, angularLimitsMin=<-45 -125 -45>, angularTarget=<0 0 0>}
      localJointOffset=<0 0 -2>
     }]
    relativeSpaceBone {boneName=root}
   }
   n6 = Component To Local(ComponentPose=@n5)
   n7 = RigidBody(ComponentPose=Local To Component(LocalPose=Use cached pose 'PreSim'())) {
    bClampLinearTranslationLimitToRefPose=true
    bOverrideWorldGravity=true
    overridePhysicsAsset=&/F_MED_Modern_Witch_FaceAcc_FE_Dyn
    simulationSpace=BaseBoneSpace
    alphaBoolBlend {blendOption=Linear}
    baseBoneRef {boneName=root}
    componentAppliedLinearAccClamp=<3000 3000 3000>
    componentLinearAccScale=<0.4 0.3 0.3>
    componentLinearVelScale=<0.4 0.3 0.3>
    overrideWorldGravity {z=-980}
   }
   n8 = Component To Local(ComponentPose=@n7)
   n9 = Get bIsInFrontEnd()
   n10 = Anim Dynamics - Bone: dyn_hat_strap_fr(
    ComponentPose=Local To Component(LocalPose=Blend Poses by bool(BlendPose_0=@n8, BlendPose_1=@n6, bActiveValue=@n9) {blendType=Linear})
    Alpha=0.75
    ExternalForce="0,0,0"
   ) {
    angularDampingOverride=0.9999
    angularSpringConstant=100
    bGravityOverrideInSimSpace=true
    bOverrideAngularDamping=true
    bUseGravityOverride=true
    alphaBoolBlend {blendOption=Linear}
    boundBone {boneName=dyn_hat_strap_fr}
    componentAppliedLinearAccClamp=<2000 2000 7000>
    componentLinearAccScale=<0.5 0.5 0.5>
    componentLinearVelScale=<0.5 0.5 0.5>
    gravityOverride {z=-980}
    physicsBodyDefinitions [{
      sphereCollisionRadius=0
      boundBone {boneName=dyn_hat_strap_fr}
      boxExtents=<12 22 22>
      constraintSetup {angularLimitsMax=<10 80 80>, angularLimitsMin=<-10 -80 -80>, angularTarget=<0 0 0>}
      localJointOffset=<6 0 0>
     }]
   }
   n11 = Anim Dynamics - Bone: dyn_hat_strap_bk(ComponentPose=@n10, Alpha=0.8, ExternalForce="0,0,0") {
    angularDampingOverride=0.9999
    angularSpringConstant=25
    bAngularSpring=true
    bGravityOverrideInSimSpace=true
    bOverrideAngularDamping=true
    bUseGravityOverride=true
    alphaBoolBlend {blendOption=Linear}
    boundBone {boneName=dyn_hat_strap_bk}
    componentAppliedLinearAccClamp=<2000 2000 3500>
    componentLinearAccScale=<0.3 0.3 0.45>
    componentLinearVelScale=<0.3 0.3 0.45>
    gravityOverride {z=-800}
    physicsBodyDefinitions [{
      sphereCollisionRadius=0
      boundBone {boneName=dyn_hat_strap_bk}
      boxExtents=<8 12 12>
      constraintSetup {angularLimitsMax=<10 80 80>, angularLimitsMin=<-10 -80 -80>, angularTarget=<0 0 0>}
      localJointOffset=<5 0 0>
     }]
   }
   n12 = Anim Dynamics - Bone: dyn_hat_center_strap_MSTR(ComponentPose=@n11, Alpha=0.5, ExternalForce="0,0,0") {
    angularDampingOverride=0.9999
    bGravityOverrideInSimSpace=true
    bOverrideAngularDamping=true
    bUseGravityOverride=true
    simulationSpace=RootRelative
    alphaBoolBlend {blendOption=Linear}
    boundBone {boneName=dyn_hat_center_strap_MSTR}
    componentLinearAccScale=<1 1 1>
    componentLinearVelScale=<1 1 1>
    gravityOverride {z=-980}
    physicsBodyDefinitions [{
      sphereCollisionRadius=0
      boundBone {boneName=dyn_hat_center_strap_MSTR}
      boxExtents=<10 20 20>
      constraintSetup {angularLimitsMax=<10 7 10>, angularLimitsMin=<-10 -60 -10>, angularTarget=<0 0 0>}
      localJointOffset=<2 0 0>
     }]
   }
   n13 = Anim Dynamics - Bone: dyn_hat_center_strap_01(ComponentPose=@n12, Alpha=0.5, ExternalForce="0,0,0") {
    angularDampingOverride=0.9999
    bGravityOverrideInSimSpace=true
    bOverrideAngularDamping=true
    bUseGravityOverride=true
    simulationSpace=RootRelative
    alphaBoolBlend {blendOption=Linear}
    boundBone {boneName=dyn_hat_center_strap_01}
    componentLinearAccScale=<1 1 1>
    componentLinearVelScale=<1 1 1>
    gravityOverride {z=-980}
    physicsBodyDefinitions [{
      sphereCollisionRadius=0
      boundBone {boneName=dyn_hat_center_strap_01}
      boxExtents=<10 20 20>
      constraintSetup {angularLimitsMax=<10 7 10>, angularLimitsMin=<-10 -60 -10>, angularTarget=<0 0 0>}
      localJointOffset=<2 0 0>
     }]
   }
   n14 = Anim Dynamics - Bone: dyn_hat_center_strap_02(ComponentPose=@n13, Alpha=0.5, ExternalForce="0,0,0") {
    angularDampingOverride=0.9999
    bGravityOverrideInSimSpace=true
    bOverrideAngularDamping=true
    bUseGravityOverride=true
    simulationSpace=RootRelative
    alphaBoolBlend {blendOption=Linear}
    boundBone {boneName=dyn_hat_center_strap_02}
    componentLinearAccScale=<1 1 1>
    componentLinearVelScale=<1 1 1>
    gravityOverride {z=-980}
    physicsBodyDefinitions [{
      sphereCollisionRadius=0
      boundBone {boneName=dyn_hat_center_strap_02}
      boxExtents=<10 20 20>
      constraintSetup {angularLimitsMax=<10 7 10>, angularLimitsMin=<-10 -60 -10>, angularTarget=<0 0 0>}
      localJointOffset=<2 0 0>
     }]
   }
   Output Pose(Result=Component To Local(ComponentPose=@n14)) {layerGroup=None}
   n15 = "Constraint - dyn_hair_03_l (TT)"!(
    ComponentPose=Local To Component(LocalPose=Copy Pose From Mesh() {bUseAttachedParent=true})
    Alpha=1
    ConstraintWeights_0=1
    ConstraintWeights_1:float=1
   ) -> (Pose) {
    alphaBoolBlend {blendOption=Linear}
    boneToModify {boneName=dyn_hair_03_l}
    constraintSetup [{transformType=Translation, targetBone {boneName=dyn_hair_02_l}}, {transformType=Translation, targetBone {boneName=dyn_hair_04_l}}]
    constraintWeights [1,1]
   }
   n16 = "Constraint - dyn_hat_strap_fr_MSTR (PP)"!(ComponentPose=@n15, Alpha=1, ConstraintWeights_0=0.1926, ConstraintWeights_1:float=0.8074) -> (Pose) {
    alphaBoolBlend {blendOption=Linear}
    boneToModify {boneName=dyn_hat_strap_fr_MSTR}
    constraintSetup [{targetBone {boneName=head}}, {targetBone {boneName=dyn_hat_l}}]
    constraintWeights [1,1]
   }
   n17 = "Constraint - dyn_hat_strap_bk_MSTR (PP)"!(ComponentPose=@n16, Alpha=1, ConstraintWeights_0=0.0347, ConstraintWeights_1:float=0.9653) -> (Pose) {
    alphaBoolBlend {blendOption=Linear}
    boneToModify {boneName=dyn_hat_strap_bk_MSTR}
    constraintSetup [{targetBone {boneName=head}}, {targetBone {boneName=dyn_hat_l}}]
    constraintWeights [1,1]
   }
   n18 = "Constraint - dyn_hat_center_strap_MSTR (PPPP)"!(
    ComponentPose=@n17
    Alpha=1
    ConstraintWeights_0=0.8991
    ConstraintWeights_1:float=0
    ConstraintWeights_2:float=0.0473
    ConstraintWeights_3:float=0.0536
   ) -> (Pose) {
    alphaBoolBlend {blendOption=Linear}
    boneToModify {boneName=dyn_hat_center_strap_MSTR}
    constraintSetup [
     {targetBone {boneName=dyn_hat_center_01}}
     {targetBone {boneName=dyn_hat_bk}}
     {targetBone {boneName=head}}
     {targetBone {boneName=dyn_hat_center_02}}
    ]
    constraintWeights [1,1,1,1]
   }
   Save cached pose 'PreSim'(Pose=Component To Local(ComponentPose=@n18)) {cacheName=PreSim}
  ]}
 EventGraph {nodes [
   Event BlueprintUpdateAnimation()
   > n1 = Gravity Override State Function DEPRECATED(CharacterPartAnimInstance=Self-Reference(), GameTimeInSeconds=GetGameTimeInSeconds(), GravityOverrideParams=Get Gravity Override Params())
   > Set GravityOverride(GravityOverride=@n1)
   TryGetPawnOwner()
  ]}
]
variables [
 Gravity Override Params {type=GravityOverrideParamsStruct}
 GravityOverride {type=Vector}
]
